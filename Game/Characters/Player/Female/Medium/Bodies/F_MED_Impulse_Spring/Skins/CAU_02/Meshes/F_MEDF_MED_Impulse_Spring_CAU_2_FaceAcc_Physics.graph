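# Physics asset for the face accessory mesh of the female medium Impulse Spring CAU_02 player skin, covering its hair bones.
# Capsule bodies form left and right hairA and hairB chains, hung from the head by swing- and twist-limited constraints, mostly tighter on the right.
# Head, upper spine, right clavicle and right upper arm capsules, joined by free-rotation constraints, appear to serve as hair colliders, with no left shoulder equivalents.
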PhysicsAsset /Game/Characters/Player/Female/Medium/Bodies/F_MED_Impulse_Spring/Skins/CAU_02/Meshes/F_MEDF_MED_Impulse_Spring_CAU_2_FaceAcc_Physics
bodies {
 clavicle_r [{shapeType=Capsule, length=9.255, radius=5.3175, shapeName=clavicle_r_capsule, center=<-7.2354 0.1711 1.0976>, rotation=<rot 84.9998 0 0>}]
 hairA_01_l [{shapeType=Capsule, length=2.4556, radius=1.7745, shapeName=hairA_01_l_capsule, center=<2.2987 0 0>, rotation=<rot 90 0 0>}]
 hairA_01_r [{shapeType=Capsule, length=1.15, radius=3.0245, shapeName=hairA_01_r_capsule, center=<2.4398 0 0>, rotation=<rot 90 0 0>}]
 hairA_02_l [{shapeType=Capsule, length=2.0675, radius=1.38, shapeName=hairA_02_l_capsule, center=<1.9739 0 0>, rotation=<rot 90 0 0>}]
 hairA_02_r [{shapeType=Capsule, length=1.4375, radius=1.9425, shapeName=hairA_02_r_capsule, center=<1.8807 0 0>, rotation=<rot 90 0 0>}]
 hairA_03_l [{shapeType=Capsule, length=3.2115, radius=1.2028, shapeName=hairA_03_l_capsule, center=<2.6407 0 0>, rotation=<rot 90 0 0>}]
 hairA_03_r [{shapeType=Capsule, length=1.7, radius=1.45, shapeName=hairA_03_r_capsule, center=<2.2927 0 0>, rotation=<rot 90 0 0>}]
 hairB_01_l [{shapeType=Capsule, length=3.8295, radius=1.699, shapeName=hairB_01_l_capsule, center=<3.042 0 0>, rotation=<rot 90 0 0>}]
 hairB_01_r [{shapeType=Capsule, length=3.8172, radius=1.1512, shapeName=hairB_01_r_capsule, center=<4.6383 0 0>, rotation=<rot 90 0 0>}]
 hairB_02_r [{shapeType=Capsule, length=1.8359, radius=1.4186, shapeName=hairB_02_r_capsule, center=<1.4969 0 0>, rotation=<rot 90 0 0>}]
 head [{shapeType=Capsule, length=15.9812, radius=5.2748, rotation=<rot -90 -165.9637 -164.0347>}]
 spine_05 [{shapeType=Capsule, length=0.505, radius=26.505, shapeName=spine_05_capsule, center=<-10.7975 -3.835 0>, rotation=<rot 0 -90 0>}]
 upperarm_r [{shapeType=Capsule, length=12, radius=4.8125, shapeName=upperarm_r_capsule, center=<-7.9208 2.317 -3.3662>, rotation=<rot 90 0 0>}]
}
constraints [
 {bone1Name=hairA_01_l, bone2Name=head, swing1LimitDegrees=45, swing2LimitDegrees=20, twistLimitDegrees=10}
 {bone1Name=hairA_02_l, bone2Name=hairA_01_l, swing1LimitDegrees=45, swing2LimitDegrees=20, twistLimitDegrees=10}
 {bone1Name=hairA_03_l, bone2Name=hairA_02_l, swing1LimitDegrees=45, swing2LimitDegrees=20, twistLimitDegrees=10}
 {bone1Name=hairB_01_l, bone2Name=head, swing1LimitDegrees=35, swing2LimitDegrees=45, twistLimitDegrees=10}
 {bone1Name=hairA_01_r, bone2Name=head, swing1LimitDegrees=25, swing2LimitDegrees=10, twistLimitDegrees=1}
 {bone1Name=hairA_02_r, bone2Name=hairA_01_r, swing1LimitDegrees=45, swing2LimitDegrees=15, twistLimitDegrees=10}
 {bone1Name=hairA_03_r, bone2Name=hairA_02_r, swing1LimitDegrees=45, swing2LimitDegrees=15, twistLimitDegrees=10, twistMotion=Locked}
 {bone1Name=hairB_01_r, bone2Name=head, swing1LimitDegrees=20, swing2LimitDegrees=5, twistLimitDegrees=1}
 {bone1Name=hairB_02_r, bone2Name=hairB_01_r, swing1LimitDegrees=20, swing2LimitDegrees=20, twistLimitDegrees=1}
 {
  bone1Name=upperarm_r
  bone2Name=clavicle_r
  swing1LimitDegrees=45
  swing1Motion=Free
  swing2LimitDegrees=45
  swing2Motion=Free
  twistLimitDegrees=45
  twistMotion=Free
 }
 {
  bone1Name=head
  bone2Name=spine_05
  swing1LimitDegrees=45
  swing1Motion=Free
  swing2LimitDegrees=45
  swing2Motion=Free
  twistLimitDegrees=45
  twistMotion=Free
 }
 {
  bone1Name=clavicle_r
  bone2Name=spine_05
  swing1LimitDegrees=45
  swing1Motion=Free
  swing2LimitDegrees=45
  swing2Motion=Free
  twistLimitDegrees=45
  twistMotion=Free
 }
]
